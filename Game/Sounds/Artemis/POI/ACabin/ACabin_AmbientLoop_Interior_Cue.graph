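# Sound cue for the looping interior ambience of the ACabin point of interest.
# A mixer layers four looping waves (main bed, wind, sub and wood creaks), with wind the most prominent.
# The wood creaks pass through a modulator that fixes their pitch lower.
# It uses the Athena indoor ambient sound class and ambient attenuation settings.
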SoundCue /Game/Sounds/Artemis/POI/ACabin/ACabin_AmbientLoop_Interior_Cue
root=Mixer_0
nodes [
 {id=Mixer_0, children [WavePlayer_0,WavePlayer_4,WavePlayer_1,Modulator_0], properties {inputVolume [0.3,0.9,0.2,0.4]}}
 {id=Modulator_0, children [WavePlayer_5], properties {pitchMax=0.8, pitchMin=0.8}}
 {id=WavePlayer_0, properties {bLooping=true, soundWaveAssetPtr=&/ACabin_AmbientLoop_Interior_Main_01}}
 {id=WavePlayer_1, properties {bLooping=true, soundWaveAssetPtr=&/ACabin_AmbientLoop_Interior_Sub_01}}
 {id=WavePlayer_4, properties {bLooping=true, soundWaveAssetPtr=&/ACabin_AmbientLoop_Interior_Wind_01}}
 {id=WavePlayer_5, properties {bLooping=true, soundWaveAssetPtr=&/ACabin_AmbientLoop_Interior_Woodcreaks_01}}
]
properties {
 attenuationSettings=&Fort_Ambient_Athena_Att
 maxDistance=2097152
 soundClassObject=&Fort_Ambient_Indoors_Athena
 volumeMultiplier=1.253
 preEffectBusSends [{}]
}
