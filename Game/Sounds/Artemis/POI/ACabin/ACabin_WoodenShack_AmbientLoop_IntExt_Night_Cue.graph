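# Looping night ambience cue for the ACabin wooden shack POI, built from exterior loops and routed to the indoor ambient sound class.
# A mixer layers main, sub, cricket and demon-tone loops, with crickets and sub quieter.
# Two nested, endlessly looping envelopers of different lengths add slow pitch drift and a gentle volume swell.
# All quality levels share this chain; a concurrency override caps instances, stopping the oldest.
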SoundCue /Game/Sounds/Artemis/POI/ACabin/ACabin_WoodenShack_AmbientLoop_IntExt_Night_Cue
root=QualityLevel_0
nodes [
 {
  id=Enveloper_1
  children [Mixer_0]
  properties {
   bLoop=true
   bLoopIndefinitely=true
   loopEnd=30
   pitchCurve {editorCurveData {keys [
      {arriveTangent=-0.0013, interpMode=RCIM_Cubic, leaveTangent=-0.0013, time=0}
      <at 15.9468 0.98 RCIM_Cubic>
      {arriveTangent=0.0014, interpMode=RCIM_Cubic, leaveTangent=0.0014, time=30}
     ]}}
   volumeCurve {editorCurveData {keys [<at 0 1>]}}
  }
 }
 {
  id=Enveloper_2
  children [Enveloper_1]
  properties {
   bLoop=true
   bLoopIndefinitely=true
   loopEnd=20
   pitchCurve {editorCurveData {keys [{time=0}]}}
   volumeCurve {editorCurveData {keys [
      {arriveTangent=-0.01, interpMode=RCIM_Cubic, leaveTangent=-0.01, time=0, value=1}
      <at 10 0.9 RCIM_Cubic>
      {arriveTangent=0.01, interpMode=RCIM_Cubic, leaveTangent=0.01, time=20, value=1}
     ]}}
  }
 }
 {id=Mixer_0, children [WavePlayer_0,WavePlayer_4,WavePlayer_1,WavePlayer_5], properties {inputVolume [0.8,0.3,0.3,0.8]}}
 {id=QualityLevel_0, children [Enveloper_2,Enveloper_2,Enveloper_2,Enveloper_2]}
 {id=WavePlayer_0, properties {bLooping=true, soundWaveAssetPtr=&/ACabin_AmbientLoop_Exterior_Main_02}}
 {id=WavePlayer_1, properties {bLooping=true, soundWaveAssetPtr=&/ACabin_AmbientLoop_Exterior_Sub_01}}
 {id=WavePlayer_4, properties {bLooping=true, soundWaveAssetPtr=&/ACabin_AmbientLoop_Exterior_Crickets_01}}
 {id=WavePlayer_5, properties {bLooping=true, soundWaveAssetPtr=&/ACabin_AmbientLoop_Exterior_DemonTone_01}}
]
properties {
 attenuationSettings=&Fort_Ambient_Athena_Att
 bOverrideConcurrency=true
 maxDistance=2097152
 soundClassObject=&Fort_Ambient_Indoors_Athena
 volumeMultiplier=1.002
 concurrencyOverrides {maxCount=2, resolutionRule=StopOldest}
 preEffectBusSends [{}]
}
